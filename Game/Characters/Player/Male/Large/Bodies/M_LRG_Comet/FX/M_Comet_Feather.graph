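# M_Comet_Feather is a masked, lit material for Niagara mesh particles, placed among the Comet player body's FX.
# An Erosion group samples an electric noise texture on the second UV channel, with adjustable scale and a parameter that blends toward swapped UV axes.
# The scaled noise feeds a dynamic-parameter lerp and a subtract node.
# These and a particle color node show no connection to material outputs.
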Material /Game/Characters/Player/Male/Large/Bodies/M_LRG_Comet/FX/M_Comet_Feather
graph {
 properties {bUsedWithNiagaraMeshParticles=true, blendMode=BLEND_Masked, pixelDepthOffsetMode=PDOM_Legacy, usedShadingModels=Default Lit}
 nodes [
  Constant()
  Constant()
  n1 = Multiply(
   A=TextureCoordinate() {coordinateIndex=1}
   B=ComponentMask(None=VectorParameter() {group=Erosion, parameterName=Noise_UV_Scale, sortPriority=4, defaultValue=<rgba 2 2 0 0>}.RGB) {g=true, r=true}
  )
  n2 = TextureSampleParameter2D(UVs=LinearInterpolate(
    A=@n1
    B=MaterialFunctionCall(XY=@n1) {materialFunction=&Swizzle}.YX
    Alpha=ScalarParameter() {group=Erosion, parameterName=Swizzle Erosion Noise, sortPriority=5}
   )) {
   group=Erosion
   parameterName=Erosion Noise
   samplerType=SAMPLERTYPE_LinearColor
   sortPriority=3
   texture=/Game/Characters/Player/Male/Medium/Bodies/M_MED_SRDrift/Textures/T_ElectricNoise
  }
  n3 = Multiply(A=@n2.R, B=ScalarParameter() {defaultValue=1, group=Erosion, parameterName=NoiseAffectMask, sortPriority=2})
  LinearInterpolate(A=Constant() {r=1}, B=@n3, Alpha=DynamicParameter() {defaultValue=<rgba 1 1 0 1>}.Param3)
  ParticleColor()
  Subtract(A=@n3)
 ]
}
