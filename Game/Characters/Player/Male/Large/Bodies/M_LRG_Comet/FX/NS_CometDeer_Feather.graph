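# Niagara effect that sheds tumbling mesh feathers from bones or sockets on the Comet Deer player character's skeletal mesh.
# Three emitters spawn large and small feathers by horizontal distance travelled, while two more spawn instantaneous bursts.
# User bools suppress all spawning while skydiving, and only the larger feathers in frontend preview.
# Feathers fall with drag, changing through scale and color or material-parameter curves; small trail feathers also use a sphere blocking volume.
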NiagaraSystem /Game/Characters/Player/Male/Large/Bodies/M_LRG_Comet/FX/NS_CometDeer_Feather
dependencies {usedDataInterfaces [&NiagaraDataInterfaceCurve], usedDynamicInputs [], usedModules [&SystemState]}
emitters {
 FeatherB {
  emitterUpdateScript [
   &EmitterState()
   &SpawnPerUnit(
    Module.Use Max Movement Threshold=true
    Module.Use Movement Tolerance=true
    Movement Tolerance=0.01
    Spawn Probability:NiagaraFloat=&MakeCustomFloatFromBool(
     Bool=$User.bIsFrontendPreview
     True Float=0
     False Float=&MakeCustomFloatFromBool(Bool=$User.IsSkydiving, True Float=0, False Float=1)
    )
    Spawn Spacing:NiagaraFloat=&RandomRangeFloat(Minimum=700, Maximum=800, Evaluation Type=::NewEnumerator1)
    Use Spawn Probability=true
    Velocity Vector:Vector3f=&Multiply_Vector(A=$Engine.Owner.Velocity, B=<1 1 0>)
   )
  ]
  particleSpawnScript [
   /SetVariables(Particles.MeshOrientation=&MakeQuatFromAxisAngle(Axis=&RandomRangeVector(Minimum=<-0.1 0.5 0.225>, Maximum=<0.1 0.5 0.275>), AngleInDegrees=90)) {isSetParameters=true}
   /SetVariables(Particles.Acceleration:Vector3f=&RandomRangeVector(Minimum=<0 0 -200>, Maximum=<0 0 -400>)) {isSetParameters=true}
   /SetVariables(Particles.SpawnRotationVector=&RandomRangeVector(Minimum=<0 0 0>, Maximum=<1 1 1>)) {isSetParameters=true}
   &InitializeParticle(
    Color Mode=::NewEnumerator0
    Lifetime=&RandomRangeFloat(Minimum=0.25, Maximum=1.5)
    Mesh Scale=/Niagara/DynamicInputs/CascadeConversion/SplitVector(
     Double Component Vector Channels:ECascadeNiagaraTwoVectorChannels=::NewEnumerator1
     Double Component:NiagaraFloat=0.9
     Single Component:NiagaraFloat=0.3
    )
    Mesh Scale Mode=::NewEnumerator4
    Position Mode=::NewEnumerator0
    Write Material Random=true
   )
   &SkeletalMeshLocation(
    Bone / Socket Index=&RandomRangeInt(Minimum=6, Maximum=9)
    Bone Sampling Mode=::NewEnumerator4
    MaskSpawnGroup=false
    Position Sampling=::NewEnumerator1
    Skeletal Mesh=$User.SkeletalMesh
    Source Movement Scale:Vector3f=<1 1 0>
   )
   &ShapeLocation(MaskSpawnGroup=false, Override Seed=false, Sphere Radius=15, Sphere Surface Distribution=0.7, UseOwnerScale=true)
   &AddVelocity(Velocity=&Multiply_VectorByFloat(Vector=$Particles.ShapeLocation.ShapeVector, Float=&RandomRangeFloat(Minimum=6, Maximum=9)))
   &Color(Color=<rgba 0.8302 0.8302 0.8302 1>)
  ]
  particleUpdateScript [
   &ParticleState()
   &ScaleMeshSize(Scale Factor=&VectorFromFloat(Value=&FloatFromCurve(FloatCurve=DataInterface() {propertyValues {Curve {keys [
          {arriveTangent=10.6107, interpMode=RCIM_Cubic, time=0, value=0}
          {arriveTangent=3.2834, interpMode=RCIM_Cubic, leaveTangent=3.2834, time=0.0986, value=1}
          <at 0.3046 1 RCIM_Cubic>
          {arriveTangent=-1.4379, interpMode=RCIM_Cubic, leaveTangent=-1.4379, time=0.75, value=1}
          <at 1 0 RCIM_Cubic>
         ]}}}))) {module=ScaleMeshSize001}
   &UpdateMeshOrientation(Rotation Vector=$Particles.SpawnRotationVector)
   &ScaleMeshSize(Scale Factor=<0.9 0.9 0.9>)
   &WindForce(Wind Speed=&RandomRangeVector(Minimum=<-10 10 -10>, Maximum=<10 20 -20>))
   &GravityForce(Gravity=&RandomRangeVector(Minimum=<0 0 -200>, Maximum=<0 0 -400>))
   &AerodynamicDrag(Aerodynamic Drag=2, Use Linear Drag=true, Use Rotational Drag=true)
   &SolveForcesAndVelocity(Clamp Velocity=false, Limit Acceleration=false)
   &DynamicMaterialParameters(
    Index 0 Param 2=&FloatFromCurve(FloatCurve=DataInterface() {propertyValues {Curve {keys [<at 0.6 1 RCIM_Cubic>,<at 1 0 RCIM_Cubic>]}}})
    Index 0 Param 3=&FloatFromCurve(FloatCurve=DataInterface() {propertyValues {Curve {keys [<at 0 0>,<at 0.8 0>,<at 1 1>]}}})
    Module.Param0WriteEnabled=false
    Module.Param1WriteEnabled=false
    Module.Param2WriteEnabled=true
    Module.Param3WriteEnabled=false
   ) {module=DynamicMaterialParameters001}
   &AlignSpriteToMeshOrientation(SetAlignment=true, SetFacing=true)
  ]
  rendererClasses [&NiagaraMeshRendererProperties]
  renderers [{rendererClass=&NiagaraMeshRendererProperties, rendererIndex=0}]
 }
 FeatherR {
  emitterUpdateScript [
   &EmitterState()
   &SpawnPerUnit(
    Module.Use Max Movement Threshold=true
    Module.Use Movement Tolerance=true
    Movement Tolerance=0.01
    Spawn Probability:NiagaraFloat=&MakeCustomFloatFromBool(
     Bool=$User.bIsFrontendPreview
     True Float=0
     False Float=&MakeCustomFloatFromBool(Bool=$User.IsSkydiving, True Float=0, False Float=1)
    )
    Spawn Spacing:NiagaraFloat=&RandomRangeFloat(Minimum=700, Maximum=800, Evaluation Type=::NewEnumerator1)
    Use Spawn Probability=true
    Velocity Vector:Vector3f=&Multiply_Vector(A=$Engine.Owner.Velocity, B=<1 1 0>)
   )
  ]
  particleSpawnScript [
   /SetVariables(Particles.MeshOrientation=&MakeQuatFromAxisAngle(Axis=&RandomRangeVector(Minimum=<-0.1 0.5 0.225>, Maximum=<0.1 0.5 0.275>), AngleInDegrees=90)) {isSetParameters=true}
   /SetVariables(Particles.Acceleration:Vector3f=&RandomRangeVector(Minimum=<0 0 -300>, Maximum=<0 0 -400>)) {isSetParameters=true}
   /SetVariables(Particles.SpawnRotationVector=&RandomRangeVector(Minimum=<0 0 0>, Maximum=<1 1 1>)) {isSetParameters=true}
   &InitializeParticle(
    Lifetime=&RandomRangeFloat(Minimum=0.25, Maximum=1.5)
    Mesh Scale=&SplitVector(
     Double Component Vector Channels:ECascadeNiagaraTwoVectorChannels=::NewEnumerator1
     Double Component:NiagaraFloat=1
     Single Component:NiagaraFloat=0.4
    )
    Mesh Scale Mode=::NewEnumerator4
    Position Mode=::NewEnumerator0
    Sprite Size Mode=::NewEnumerator3
    Write Material Random=true
   )
   &SkeletalMeshLocation(
    Bone / Socket Index=&RandomRangeInt(Minimum=0, Maximum=5)
    Bone Sampling Mode=::NewEnumerator4
    MaskSpawnGroup=false
    Position Sampling=::NewEnumerator1
    Skeletal Mesh=$User.SkeletalMesh
    Source Movement Scale:Vector3f=<1 1 0>
   )
   &ShapeLocation(
    Hemisphere Distribution=<1 0.5>
    MaskSpawnGroup=false
    Non Uniform Scale=<1 1 -1>
    Override Seed=false
    Sphere Radius=15
    Sphere Surface Distribution=0.7
    UseOwnerScale=true
   )
   &AddVelocity(Velocity=&Multiply_VectorByFloat(Vector=$Particles.ShapeLocation.ShapeVector, Float=&RandomRangeFloat(Minimum=6, Maximum=9)))
   &DynamicMaterialParameters(
    Index 0 Param 1=&RandomRangeFloat(Minimum=0.5, Maximum=0.9)
    Index 0 Param 4=&RandomRangeFloat(Minimum=0.5, Maximum=1)
    Module.Param0WriteEnabled=true
    Module.Param1WriteEnabled=false
    Module.Param2WriteEnabled=false
    Module.Param3WriteEnabled=true
   )
   &Color(Color=<rgba 0.3635 0.2116 0.1632 1>)
  ]
  particleUpdateScript [
   &ParticleState()
   &ScaleMeshSize(Scale Factor=&VectorFromFloat(Value=&FloatFromCurve(FloatCurve=DataInterface() {propertyValues {Curve {keys [
          {arriveTangent=10.6107, interpMode=RCIM_Cubic, time=0, value=0}
          {arriveTangent=3.2834, interpMode=RCIM_Cubic, leaveTangent=3.2834, time=0.0986, value=1}
          <at 0.3046 1 RCIM_Cubic>
          {arriveTangent=-1.4379, interpMode=RCIM_Cubic, leaveTangent=-1.4379, time=0.75, value=1}
          <at 1 0 RCIM_Cubic>
         ]}}})))
   &UpdateMeshOrientation(Rotation Vector=$Particles.SpawnRotationVector)
   &GravityForce(Gravity=&RandomRangeVector(Minimum=<0 0 -200>, Maximum=<0 0 -400>))
   &WindForce(Wind Speed=&RandomRangeVector(Minimum=<-10 10 -10>, Maximum=<10 20 -20>))
   &AerodynamicDrag(Aerodynamic Drag=2, Use Linear Drag=true, Use Rotational Drag=true)
   &SolveForcesAndVelocity(Clamp Velocity=false, Limit Acceleration=false)
   &DynamicMaterialParameters(
    Index 0 Param 2=&FloatFromCurve(FloatCurve=DataInterface() {propertyValues {Curve {keys [<at 0.6 1 RCIM_Cubic>,<at 1 0 RCIM_Cubic>]}}})
    Index 0 Param 3=&FloatFromCurve(FloatCurve=DataInterface() {propertyValues {Curve {keys [<at 0 0>,<at 0.8 0>,<at 1 1>]}}})
    Module.Param0WriteEnabled=false
    Module.Param1WriteEnabled=false
    Module.Param2WriteEnabled=true
    Module.Param3WriteEnabled=false
   ) {module=DynamicMaterialParameters001}
   &AlignSpriteToMeshOrientation(SetAlignment=true, SetFacing=true)
   &ScaleColor(
    Scale Alpha=&FloatFromCurve(FloatCurve=DataInterface() {propertyValues {Curve {keys [<at 0 1>,<at 0.6 1>,<at 1 0>]}}})
    ScaleA=true
    ScaleRGB=true
   )
  ]
  rendererClasses [&NiagaraMeshRendererProperties]
  renderers [{rendererClass=&NiagaraMeshRendererProperties, rendererIndex=0}]
 }
 FeatherRBurst {
  emitterUpdateScript [
   &EmitterState(Life Cycle Mode=::NewEnumerator1, Loop Behavior=::NewEnumerator1, Loop Duration=1, UseLoopDelay=true)
   &SpawnBurst_Instantaneous(
    Spawn Count=&MakeCustomIntFromBool(
     Bool=$User.bIsFrontendPreview
     True INT=0
     False INT=&MakeCustomIntFromBool(Bool=$User.IsSkydiving, True INT=0, False INT=&RandomRangeInt(Minimum=1, Maximum=2))
    )
    Use Spawn Probability=false
    UseLoopCountLimit=false
   )
  ]
  particleSpawnScript [
   /SetVariables(Particles.MeshOrientation=&MakeQuatFromAxisAngle(Axis=&RandomRangeVector(Minimum=<-0.1 0.5 0.225>, Maximum=<0.1 0.5 0.275>), AngleInDegrees=90)) {isSetParameters=true}
   /SetVariables(Particles.Acceleration:Vector3f=&RandomRangeVector(Minimum=<0 0 -150>, Maximum=<0 0 -250>)) {isSetParameters=true}
   /SetVariables(Particles.SpawnRotationVector=&RandomRangeVector(Minimum=<0 0 0>, Maximum=<1 1 1>)) {isSetParameters=true}
   &InitializeParticle(
    Lifetime=&RandomRangeFloat(Minimum=0.25, Maximum=1.5)
    Mesh Scale=&SplitVector(
     Double Component Vector Channels:ECascadeNiagaraTwoVectorChannels=::NewEnumerator1
     Double Component:NiagaraFloat=0.85
     Single Component:NiagaraFloat=0.4
    )
    Mesh Scale Mode=::NewEnumerator4
    UsePositionOffset=false
    Write Material Random=true
   )
   &SkeletalMeshLocation(
    Bone / Socket Index=&RandomRangeInt(Minimum=0, Maximum=5)
    Bone Sampling Mode=::NewEnumerator4
    MaskSpawnGroup=false
    Skeletal Mesh=$User.SkeletalMesh
    Velocity Sampling=::NewEnumerator0
    VelocitySpeedLimit=false
   )
   &ShapeLocation(
    Hemisphere Distribution=<1 0.5>
    MaskSpawnGroup=false
    Override Seed=false
    Sphere Radius=15
    Sphere Surface Distribution=0.7
    UseOwnerScale=true
   )
   &AddVelocity(
    Velocity=&Multiply_VectorByFloat(Vector=$Particles.ShapeLocation.ShapeVector, Float=&RandomRangeFloat(Minimum=5, Maximum=10))
    Velocity Speed Scale=1.5
   )
   &DynamicMaterialParameters(
    Index 0 Param 1=&RandomRangeFloat(Minimum=0.5, Maximum=0.9)
    Index 0 Param 4=&RandomRangeFloat(Minimum=0.5, Maximum=1)
    Module.Param0WriteEnabled=true
    Module.Param1WriteEnabled=false
    Module.Param2WriteEnabled=false
    Module.Param3WriteEnabled=true
   )
   &Color(Color=<rgba 0.8302 0.8302 0.8302 1>)
  ]
  particleUpdateScript [
   &ParticleState()
   &ScaleMeshSize(Scale Factor=&VectorFromFloat(Value=&FloatFromCurve(FloatCurve=DataInterface() {propertyValues {Curve {keys [
          {arriveTangent=10.6107, interpMode=RCIM_Cubic, time=0, value=0}
          {arriveTangent=3.2834, interpMode=RCIM_Cubic, leaveTangent=3.2834, time=0.0986, value=1}
          <at 0.3046 1 RCIM_Cubic>
          {arriveTangent=-1.4379, interpMode=RCIM_Cubic, leaveTangent=-1.4379, time=0.75, value=1}
          <at 1 0 RCIM_Cubic>
         ]}}})))
   &UpdateMeshOrientation(Rotation Vector=$Particles.SpawnRotationVector)
   &AccelerationForce(Acceleration=$Particles.Acceleration)
   &AerodynamicDrag(Use Linear Drag=true, Use Rotational Drag=true)
   &SolveForcesAndVelocity(Clamp Velocity=false, Limit Acceleration=false)
   &DynamicMaterialParameters(
    Index 0 Param 2=&FloatFromCurve(FloatCurve=DataInterface() {propertyValues {Curve {keys [<at 0.6 1 RCIM_Cubic>,<at 1 0 RCIM_Cubic>]}}})
    Index 0 Param 3=&FloatFromCurve(FloatCurve=DataInterface() {propertyValues {Curve {keys [<at 0 0>,<at 0.8 0>,<at 1 1>]}}})
    Module.Param0WriteEnabled=false
    Module.Param1WriteEnabled=true
    Module.Param2WriteEnabled=true
    Module.Param3WriteEnabled=false
   ) {module=DynamicMaterialParameters001}
   &AlignSpriteToMeshOrientation(SetAlignment=true, SetFacing=true)
   &ScaleColor(
    Scale Alpha=&FloatFromCurve(FloatCurve=DataInterface() {propertyValues {Curve {keys [<at 0 1>,<at 0.6 1>,<at 1 0>]}}})
    ScaleA=true
    ScaleRGB=true
   )
  ]
  rendererClasses [&NiagaraMeshRendererProperties]
  renderers [{rendererClass=&NiagaraMeshRendererProperties, rendererIndex=0}]
 }
 FeatherSmall {
  emitterUpdateScript [
   &EmitterState()
   &SpawnPerUnit(
    Module.Use Max Movement Threshold=true
    Module.Use Movement Tolerance=true
    Movement Tolerance=0.01
    Spawn Probability:NiagaraFloat=&MakeCustomFloatFromBool(Bool=$User.IsSkydiving, True Float=0, False Float=1)
    Spawn Spacing:NiagaraFloat=&RandomRangeFloat(Minimum=150, Maximum=200, Evaluation Type=::NewEnumerator1)
    Use Spawn Probability=true
    Velocity Vector:Vector3f=&Multiply_Vector(A=$Engine.Owner.Velocity, B=<1 1 0>)
   )
  ]
  particleSpawnScript [
   /SetVariables(Particles.MeshOrientation=&MakeQuatFromAxisAngle(Axis=&RandomRangeVector(Minimum=<-5 -5 -5>, Maximum=<5 5 5>), AngleInDegrees=90)) {isSetParameters=true}
   /SetVariables(Particles.SpawnRotationVector=&RandomRangeVector(Minimum=<-1.5 -1.5 -1.5>, Maximum=<1.5 2 1.5>)) {isSetParameters=true}
   /SetVariables(Particles.Acceleration:Vector3f=&RandomRangeVector(Minimum=<0 0 -110>, Maximum=<0 0 -80>)) {isSetParameters=true}
   &InitializeParticle(
    Lifetime=&RandomRangeFloat(Minimum=0.75, Maximum=1.25)
    Mesh Scale=&RandomRangeVector(Minimum=<0.4 0.15 0.3>, Maximum=<0.5 0.2 0.75>)
    Mesh Scale Mode=::NewEnumerator4
    Position=&ConvertVectorToPosition(
     Input Position=&RandomRangeVector(Minimum=<0 0 100>, Maximum=<0 0 130>)
     Conversion Operation:ENiagara_LWCConvertVecToPos=::NewEnumerator0
    )
    Position Mode=::NewEnumerator1
    UsePositionOffset=false
    Write Material Random=true
   )
   &SkeletalMeshLocation(
    Bone / Socket Index=&RandomRangeInt(Minimum=0, Maximum=5)
    Bone Sampling Mode=::NewEnumerator4
    MaskSpawnGroup=false
    Position Sampling=::NewEnumerator1
    Skeletal Mesh=$User.SkeletalMesh
    Source Movement Scale:Vector3f=<1 1 0>
   )
   &ShapeLocation(MaskSpawnGroup=false, Override Seed=false, Sphere Radius=15, Sphere Surface Distribution=0.7, UseOwnerScale=true)
   &AddVelocity(Velocity=&Multiply_VectorByFloat(Vector=$Particles.ShapeLocation.ShapeVector, Float=&RandomRangeFloat(Minimum=7, Maximum=12)))
  ]
  particleUpdateScript [
   &ParticleState()
   &ScaleMeshSize(Scale Factor=&VectorFromFloat(Value=&FloatFromCurve(FloatCurve=DataInterface() {propertyValues {Curve {keys [
          {arriveTangent=8.4731, interpMode=RCIM_Cubic, time=-0.0012, value=0.5172}
          {arriveTangent=1.2764, interpMode=RCIM_Cubic, leaveTangent=1.2764, time=0.1481, value=1.3202}
          {arriveTangent=-1.5513, interpMode=RCIM_Cubic, leaveTangent=-1.5513, time=0.797, value=1.5361}
          {interpMode=RCIM_Cubic, leaveTangent=-10.3217, time=0.9992, value=0}
         ]}}}))) {module=ScaleMeshSize001}
   &UpdateMeshOrientation(Rotation Vector=$Particles.SpawnRotationVector)
   &ScaleMeshSize(Scale Factor=<0.1 0.1 0.1>)
   &WindForce(Wind Speed=&RandomRangeVector(Minimum=<-10 10 -10>, Maximum=<10 20 -20>))
   &GravityForce(Gravity=&RandomRangeVector(Minimum=<0 0 -200>, Maximum=<0 0 -400>))
   &AerodynamicDrag(Use Linear Drag=true, Use Rotational Drag=true)
   &SolveForcesAndVelocity(Clamp Velocity=false, Limit Acceleration=false)
   &AlignSpriteToMeshOrientation(SetAlignment=true, SetFacing=true)
   &Color(Color=&MakeLinearColorFromVectorAndFloat(
     "Vector (RGB)"=&MakeVectorFromLinearColorRGB(Color=<rgba 0.3635 0.2116 0.1632 1>)
     "Float (Alpha)"=&ClampFloat(Float=&FloatFromCurve(FloatCurve=DataInterface() {propertyValues {Curve {keys [<at 0 1>,<at 1 0>]}}}))
    ))
   &self:SphereBlockingVolume(Pivot Offset:Vector3f=<30 0 50>, Sphere Origin:NiagaraPosition=$Engine.Emitter.SimulationPosition, Sphere Radius=120) {module=SphereBlockingVolume}
  ]
  rendererClasses [&NiagaraMeshRendererProperties]
  renderers [{rendererClass=&NiagaraMeshRendererProperties, rendererIndex=0}]
 }
 FeatherSmallBurst {
  emitterUpdateScript [
   &EmitterState(Life Cycle Mode=::NewEnumerator1, Loop Behavior=::NewEnumerator1, Loop Duration=1, UseLoopDelay=true)
   &SpawnBurst_Instantaneous(
    Spawn Count=&MakeCustomIntFromBool(Bool=$User.IsSkydiving, True INT=0, False INT=&RandomRangeInt(Minimum=6, Maximum=8))
    Use Spawn Probability=false
    UseLoopCountLimit=false
   )
  ]
  particleSpawnScript [
   /SetVariables(Particles.MeshOrientation=&MakeQuatFromAxisAngle(Axis=&RandomRangeVector(Minimum=<-5 -5 -5>, Maximum=<5 5 5>), AngleInDegrees=90)) {isSetParameters=true}
   /SetVariables(Particles.SpawnRotationVector=&RandomRangeVector(Minimum=<-1.5 -1.5 -1.5>, Maximum=<1.5 2 1.5>)) {isSetParameters=true}
   /SetVariables(Particles.Acceleration:Vector3f=&RandomRangeVector(Minimum=<0 0 -100>, Maximum=<0 0 -200>)) {isSetParameters=true}
   &InitializeParticle(
    Lifetime=&RandomRangeFloat(Minimum=0.5, Maximum=1.2)
    Mesh Scale=&RandomRangeVector(Minimum=<0.25 0.15 0.3>, Maximum=<0.5 0.2 0.75>)
    Mesh Scale Mode=::NewEnumerator4
    UsePositionOffset=false
    Write Material Random=true
   )
   &SkeletalMeshLocation(
    Bone / Socket Index=&RandomRangeInt(Minimum=0, Maximum=5)
    Bone Sampling Mode=::NewEnumerator4
    MaskSpawnGroup=false
    Skeletal Mesh=$User.SkeletalMesh
    Velocity Sampling=::NewEnumerator0
    VelocitySpeedLimit=false
   )
   &ShapeLocation(
    MaskSpawnGroup=false
    Offset Mode=::NewEnumerator0
    Override Seed=false
    Rotation Axis=<0 0 1>
    Rotation Mode=::NewEnumerator1
    Sphere Radius=15
    Sphere Surface Distribution=0.7
    UseOwnerScale=true
   )
   &AddVelocity(
    Velocity=&Multiply_VectorByFloat(Vector=$Particles.ShapeLocation.ShapeVector, Float=&RandomRangeFloat(Minimum=5, Maximum=10))
    Velocity Speed Scale=1.8
   )
  ]
  particleUpdateScript [
   &ParticleState()
   &ScaleMeshSize(Scale Factor=&VectorFromFloat(Value=&FloatFromCurve(FloatCurve=DataInterface() {propertyValues {Curve {keys [
          {arriveTangent=8.4731, interpMode=RCIM_Cubic, time=-0.0012, value=0.5172}
          {arriveTangent=-0.5166, interpMode=RCIM_Cubic, leaveTangent=-0.5166, time=0.1481, value=1.3202}
          {interpMode=RCIM_Cubic, leaveTangent=-0.017, time=1, value=0}
         ]}}}))) {module=ScaleMeshSize001}
   &UpdateMeshOrientation(Rotation Vector=$Particles.SpawnRotationVector)
   &ScaleMeshSize(Scale Factor=<0.17 0.17 0.17>)
   &AccelerationForce(Acceleration=$Particles.Acceleration)
   &AerodynamicDrag(Use Linear Drag=true, Use Rotational Drag=true)
   &SolveForcesAndVelocity(Clamp Velocity=false, Limit Acceleration=false)
   &AlignSpriteToMeshOrientation(SetAlignment=true, SetFacing=true)
   &Color(Color=&MakeLinearColorFromVectorAndFloat(
     "Vector (RGB)"=&MakeVectorFromLinearColorRGB(Color=<rgba 0.3635 0.2116 0.1632 1>)
     "Float (Alpha)"=&ClampFloat(Float=&FloatFromCurve(FloatCurve=DataInterface() {propertyValues {Curve {keys [<at 0 1>,<at 1 0>]}}}))
    ))
  ]
  rendererClasses [&NiagaraMeshRendererProperties]
  renderers [{rendererClass=&NiagaraMeshRendererProperties, rendererIndex=0}]
 }
}
summary {
 systemName=NS_CometDeer_Feather
 emitters [
  {emitterName=FeatherR, rendererClasses [&NiagaraMeshRendererProperties]}
  {emitterName=FeatherRBurst, rendererClasses [&NiagaraMeshRendererProperties]}
  {emitterName=FeatherB, rendererClasses [&NiagaraMeshRendererProperties]}
  {emitterName=FeatherSmall, rendererClasses [&NiagaraMeshRendererProperties]}
  {emitterName=FeatherSmallBurst, rendererClasses [&NiagaraMeshRendererProperties]}
 ]
 userVariables [
  User.bIsFrontendPreview {type=NiagaraBool}
  User.IsSkydiving {type=NiagaraBool}
  User.SkeletalMesh {
   type=NiagaraDataInterfaceSkeletalMesh
   defaultValue {
    struct=NiagaraExt_VariableValue_DataInterface
    value {dataInterface=&self:NiagaraDataInterfaceSkeletalMesh_0}
   }
  }
 ]
}
